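# Additive unlit particle material for the enraged Husk eye effect in mutators, usable on sprites and mesh particles, including Niagara.
# A switchable two-color Fresnel tint multiplies an optional emissive texture and particle color.
# An optional time-of-day blend scales that color with per-period multipliers before it passes through a shared emissive adjustment function.
# Particle alpha is multiplied by a depth fade to soften intersections.
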Material /Game/Effects/Fort_Effects/Mutators/Materials/M_Husk_Enraged_Eyes_01
graph {
 properties {
  bAllowNegativeEmissiveColor=true
  bUsedWithMeshParticles=true
  bUsedWithNiagaraMeshParticles=true
  bUsedWithNiagaraSprites=true
  bUsedWithParticleSprites=true
  blendMode=BLEND_Additive
  pixelDepthOffsetMode=PDOM_Legacy
  shadingModel=MSM_Unlit
  usedShadingModels=Unlit
 }
 nodes [
  n1 = TextureSampleParameter2D() {group=Emissive, parameterName=EmissiveTexture, texture=/Game/Weapons/FORT_Melee/Pickaxe_NeonGraffiti/Textures/T_Pickaxe_Neon_Graffiti_E}
  n2 = ScalarParameter() {parameterName=UseEmissiveTexture}
  n3 = LinearInterpolate(
   A=VectorParameter() {parameterName=Color A}.RGB
   B=VectorParameter() {parameterName=Color B, defaultValue=<rgba 15 0.675 0.675 0>}.RGB
   Alpha=MaterialFunctionCall(Exponent=ScalarParameter() {defaultValue=3, parameterName=Fresnel Exponent}) {materialFunction=&MF_Fresnel_FX}.Fresnel
  )
  n4 = Constant() {r=1}
  n5 = ParticleColor()
  n6 = Multiply(
   A=Multiply(
    A=LinearInterpolate(B=@n1.RGB, Alpha=@n2) {constA=1}
    B=StaticSwitchParameter(True=@n3, False=@n4) {defaultValue=true, parameterName=UseFresnelColor}
   )
   B=@n5.RGB
  )
  n7 = Multiply(A=ScalarParameter() {defaultValue=1, group=Color_TimeOfDay, parameterName=MorningColorMult}, B=@n6)
  n8 = Multiply(A=ScalarParameter() {defaultValue=1, group=Color_TimeOfDay, parameterName=DayColorMult}, B=@n6)
  n9 = Multiply(A=@n6, B=ScalarParameter() {defaultValue=1, group=Color_TimeOfDay, parameterName=EveningColorMult})
  n10 = Multiply(A=@n6, B=ScalarParameter() {defaultValue=1, group=Color_TimeOfDay, parameterName=NIghtColorMult})
  MaterialFunctionCall(Emissive=LinearInterpolate(
    A=@n6
    B=MaterialFunctionCall(Color_Morning=@n7, Color_Day=@n8, Color_Evening=@n9, Color_Night=@n10) {materialFunction=&Color_Time_of_Day}.Result
    Alpha=ScalarParameter() {group=Color_TimeOfDay, parameterName=UseColorTimeOfDay}
   )) {materialFunction=&MF_FX_EmissiveAdjustments}
  Multiply(A=@n5.A, B=DepthFade(FadeDistance=ScalarParameter() {defaultValue=10, parameterName=Depth Fade Distance}))
 ]
}
