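# Additive unlit material for Husk health cross mesh particles in the mutator effects, usable by both Cascade and Niagara.
# Inverted Fresnel plus vertex color, saturated, blends between inner and outer color parameters.
# The blend is tinted by particle color and passed through an emissive adjustment function.
# Particle alpha gets a depth fade on the default shading path, while mobile uses the raw alpha.
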Material /Game/Effects/Fort_Effects/Mutators/Materials/M_Husk_HealthCross_01
graph {
 properties {
  bUsedWithMeshParticles=true
  bUsedWithNiagaraMeshParticles=true
  blendMode=BLEND_Additive
  pixelDepthOffsetMode=PDOM_Legacy
  previewMesh=/Game/Characters/Enemies/Husk/Mesh/FX/St_Husk_HealthCross_01
  shadingModel=MSM_Unlit
  usedShadingModels=Unlit
 }
 nodes [
  n1 = VectorParameter() {parameterName=Color Inner, defaultValue=<rgba 0.2 0.0516 0.1798 0>}
  n2 = VectorParameter() {parameterName=Color Outer, defaultValue=<rgba 1.1085 0.5889 5 0>}
  n3 = ScalarParameter() {defaultValue=1, parameterName=Fresnel Brightness}
  n4 = ScalarParameter() {defaultValue=1, parameterName=Fresnel Exponent}
  n5 = Saturate(None=Add(
    A=MaterialFunctionCall(Brightness=@n3, Exponent=@n4) {materialFunction=&MF_Fresnel_FX}.Fresnel Inverted
    B=VertexColor()
   ))
  n6 = ParticleColor()
  MaterialFunctionCall(Emissive=Multiply(A=LinearInterpolate(A=@n1.RGB, B=@n2.RGB, Alpha=@n5), B=@n6.RGB)) {materialFunction=&MF_FX_EmissiveAdjustments}
  ShadingPathSwitch(Default=DepthFade(Opacity=@n6.A, FadeDistance=ScalarParameter() {defaultValue=20, parameterName=Depth Fade Distance}), Mobile=@n6.A) {desc=Migrated from FeatureLevelSwitch}
 ]
}
