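# Translucent unlit particle material for Volcanic Pit mutator effects, set up for refraction distortion using pixel normal offset on sprites and meshes.
# A radial sphere mask shapes refraction strength, switchable between Dynamic Parameter control and particle alpha.
# A swirl-noise normal map uses screen or mesh UVs, with switchable tiling source and optional panning.
# The panner speed feeds SpeedX into both axes, leaving SpeedY unused.
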Material /Game/Effects/Fort_Effects/Mutators/Materials/M_VolcanicPit_Refraction_01
graph {
 properties {
  bUsedWithMeshParticles=true
  bUsedWithNiagaraMeshParticles=true
  bUsedWithNiagaraSprites=true
  bUsedWithParticleSprites=true
  blendMode=BLEND_Translucent
  pixelDepthOffsetMode=PDOM_Legacy
  refractionMethod=RM_PixelNormalOffset
  shadingModel=MSM_Unlit
  usedShadingModels=Unlit
 }
 nodes [
  Constant()
  MaterialFunctionCall!(Center Point World Position) {materialFunction=/Engine/Functions/Engine_MaterialFunctions02/UVs/WorldSpaceAlignedScreenCoordinates}
  n1 = ParticleColor()
  n2 = Constant() {r=0.5}
  n3 = SphereMask(A=TextureCoordinate(), B=@n2, Radius=@n2, Hardness=Constant())
  StaticSwitchParameter(
   True=LinearInterpolate(Alpha=Multiply(A=@n1.A, B=@n3)) {constA=1, constB=1.05}
   False=LinearInterpolate(B=DynamicParameter() {defaultValue=<rgba 1.1 1 1 1>, paramNames [Refraction, UV Tile, SpeedX, SpeedY]}.Refraction, Alpha=@n3) {constA=1}
  ) {parameterName=UseParticleColorToHandleRef}
  n4 = DynamicParameter() {defaultValue=<rgba 1.1 1 1 1>, paramNames [Refraction, UV Tile, SpeedX, SpeedY]}
  n5 = VectorParameter() {parameterName=UVControl, defaultValue=<rgba 1 1 0 1>}
  n6 = Multiply(
   A=StaticSwitchParameter(True=ScreenPosition().ViewportUV, False=TextureCoordinate()) {parameterName=UseScreenUV}
   B=StaticSwitchParameter(True=@n4.UV Tile, False=ComponentMask(None=@n5.RGB) {g=true, r=true}) {defaultValue=true, parameterName=UVControlByDynParam?}
  )
  n7 = AppendVector(A=@n4.SpeedX, B=@n4.SpeedX)
  TextureSampleParameter2D(UVs=StaticSwitchParameter(True=Panner(Coordinate=@n6, Speed=@n7) {speedY=-1}, False=@n6) {parameterName=UsePanner?}) {parameterName=T_Noise, samplerType=SAMPLERTYPE_Normal, texture=/Game/Global/Textures/Default/Noise/T_SwirlNoise_N}
 ]
}
